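# Idle Niagara effect for the TicketPoke backpack cosmetic, combining icy blue frost smoke with teal glow sprites.
# FrostSmoke spawns with movement and inherits velocity, FrostSmokeLocal emits steadily; both place flipbook sprites in a six-way rotated curve pattern.
# Glow bursts eleven partitioned sprites at fixed socket offsets, with view-angle-dependent camera offset.
# Time-of-day tints color, an equip timer drives alpha, and pickaxe-equipped state collapses smoke and some glows.
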NiagaraSystem /BRCosmetics/Accessories/FORT_Backpacks/Backpack_M_MED_TicketPoke/FX/NS_Backpack_TicketPoke_Idle
dependencies {
 usedDataInterfaces [
  &NiagaraDataInterfaceVectorCurve
  &NiagaraDataInterfaceCurve
  &NiagaraDataInterfaceCamera
  &NiagaraDataInterfaceSocketReader
  &NiagaraDataInterfaceColorCurve
 ]
 usedDynamicInputs [
  &ReturnParticleIDAttributes
  &ColorFromCurve
  &Subtract_Float
 ]
 usedModules [
  &SystemState
  &self:SystemSpawnScript.SystemScriptSource.SystemScriptGraph.NiagaraNodeAssignment_0.SetVariables_#1
  &self:SystemSpawnScript.SystemScriptSource.SystemScriptGraph.NiagaraNodeAssignment_2.SetVariables_#2
  &self:SystemSpawnScript.SystemScriptSource.SystemScriptGraph.NiagaraNodeAssignment_3.SetVariables_#3
 ]
}
emitters {
 FrostSmoke {
  emitterUpdateScript [
   &EmitterState()
   &SpawnPerUnit(
    Max Movement Threshold=1000
    Module.Use Max Movement Threshold=true
    Module.Use Movement Tolerance=true
    Movement Tolerance=5
    Spawn Probability=0.7
    Use Spawn Probability=true
   )
  ]
  particleSpawnScript [
   &InitializeParticle(
    Color=&MultiplyLinearColor(
     Linear Color=&RandomRangeLinearColor(Minimum=<rgba 0.3906 0.7302 1 0.3>, Maximum=<rgba 0.5469 0.7994 1 0.3>)
     Scale Factor=&MakeVector4_FromLinearColor(LinearColor=$System.TOD)
    )
    Lifetime=&RandomRangeFloat(Minimum=0.6, Maximum=1)
    Sprite Rotation Mode=::NewEnumerator2
    Sprite Size Mode=::NewEnumerator3
    Uniform Sprite Size=&RandomRangeFloat(Minimum=25, Maximum=30)
    UsePositionOffset=false
    Write Material Random=true
   ) {module=InitializeParticle001}
   &CameraOffset(Camera Offset Amount=20)
   /SetVariables(Particles.Position=&AddVectorToPosition(
     Position=/Niagara/DynamicInputs/Helpers/SimulationPosition()
     Vector=&Add_Vector(
      A=<0 -15 3.8564>
      B=/Niagara/DynamicInputs/Vectors/RotateVector(
       Vector To Rotate:Vector3f=&VectorFromCurve(
        VectorCurve=DataInterface() {propertyValues {
          XCurve {keys [<at 0 1>,<at 1 12>]}
          YCurve {keys [<at 0 1>,<at 1 0>]}
          ZCurve {keys [
            {arriveTangent=-11.2611, interpMode=RCIM_Cubic, time=0, value=-1}
            {arriveTangent=1.9321, interpMode=RCIM_Cubic, leaveTangent=1.9321, tangentMode=RCTM_User, time=1, value=0}
           ]}
         }}
        CurveIndex=&RandomRangeFloat(Minimum=0, Maximum=1)
       )
       Yaw:NiagaraFloat=0
       Pitch:NiagaraFloat=&Multiply_FloatByInt(Float=60, Integer=&RandomRangeInt(Minimum=0, Maximum=5))
       Roll:NiagaraFloat=0
       Rotation Rate:NiagaraFloat=1
      )
     )
    )) {isSetParameters=true}
   &AddVelocity(Velocity=&RandomRangeVector(Minimum=<-1 -1 -1>, Maximum=<1 1 1>), Velocity Speed Scale=5)
   /SetVariables(Particles.StartIndex:NiagaraInt32=&RandomRangeInt(Minimum=2, Maximum=20)) {isSetParameters=true, module=SetVariables_#4}
   &InheritVelocity(Inherited Velocity Speed Limit=200, Speed Limit=true, UseSpeedThreshold=false)
  ]
  particleUpdateScript [
   &ParticleState()
   &SubUVAnimation(
    End Frame:NiagaraInt32=&Add_Integer(A=$Particles.StartIndex, B=&RandomRangeInt(Minimum=20, Maximum=30))
    Start Frame:NiagaraInt32=$Particles.StartIndex
   )
   &GravityForce(Gravity=&RandomRangeVector(Minimum=<0 0 -6>, Maximum=<0 0 -10>))
   &Drag(Drag=&RandomRangeFloat(Minimum=2, Maximum=3), Use Linear Drag=true, Use Rotational Drag=true)
   &SolveForcesAndVelocity(Clamp Velocity=false, Limit Acceleration=false)
   &ScaleColor(
    Scale Alpha=&FloatFromCurve(
     FloatCurve=DataInterface() {propertyValues {Curve {keys [
         <at 0.0122 0 RCIM_Cubic>
         {arriveTangent=0.317, interpMode=RCIM_Cubic, leaveTangent=0.317, tangentMode=RCTM_User, time=0.1, value=0.4}
         <at 1 0 RCIM_Cubic>
        ]}}}
     Scale Curve=&Multiply_Float(
      A=&Multiply_Float(
       A=&RandomRangeFloat(Minimum=0.2, Maximum=0.22)
       B=/Niagara/DynamicInputs/Camera/CameraFade(
        Camera Query:NiagaraDataInterfaceCamera=DataInterface()
        Fade Offset:NiagaraFloat=100
        Fade Distance:NiagaraFloat=200
        Use Owner Position:NiagaraBool=true
       )
      )
      B=$System.EquipTimer
     )
    )
    ScaleA=true
    ScaleRGB=true
   )
   &DynamicMaterialParameters(
    Index 0 Param 4=1
    Module.Param0WriteEnabled=true
    Module.Param1WriteEnabled=true
    Module.Param2WriteEnabled=true
    Module.Param3WriteEnabled=true
   )
   &ScaleSpriteSize(Scale Factor=&Vector2DFromFloat(Value=&FloatFromCurve(
      FloatCurve=DataInterface() {propertyValues {Curve {keys [
          {arriveTangent=3.3691, interpMode=RCIM_Cubic, leaveTangent=3.3691, tangentMode=RCTM_User, time=0, value=0.6}
          {interpMode=RCIM_Cubic, leaveTangent=0.1873, time=1, value=1.5}
         ]}}}
      Scale Curve=&MakeCustomFloatFromBool(Bool=$User.bPickaxeEquipped, True Float=0, False Float=1)
     )))
  ]
  rendererClasses [&NiagaraSpriteRendererProperties]
  renderers [{rendererClass=&NiagaraSpriteRendererProperties}]
 }
 FrostSmokeLocal {
  emitterUpdateScript [
   &EmitterState()
   &SpawnRate(SpawnRate=50, Use Spawn Probability=false)
  ]
  particleSpawnScript [
   &InitializeParticle(
    Color=&MultiplyLinearColor(
     Linear Color=&RandomRangeLinearColor(Minimum=<rgba 0.3906 0.7302 1 0.3>, Maximum=<rgba 0.5469 0.7994 1 0.3>)
     Scale Factor=&MakeVector4_FromLinearColor(LinearColor=$System.TOD)
    )
    Lifetime=&RandomRangeFloat(Minimum=1.5, Maximum=2.5)
    Sprite Rotation Mode=::NewEnumerator2
    Sprite Size Mode=::NewEnumerator3
    Uniform Sprite Size=7
    UsePositionOffset=false
    Write Material Random=true
   ) {module=InitializeParticle001}
   &CameraOffset(Camera Offset Amount=20)
   /SetVariables(Particles.Position=&AddVectorToPosition(
     Position=&RotateVector(
      Vector To Rotate:Vector3f=&VectorFromCurve(
       VectorCurve=DataInterface() {propertyValues {
         XCurve {keys [<at 0 1>,<at 1 12>]}
         YCurve {keys [<at 0 1>,<at 1 0>]}
         ZCurve {keys [
           {arriveTangent=-11.2611, interpMode=RCIM_Cubic, time=0, value=-1}
           {arriveTangent=1.9321, interpMode=RCIM_Cubic, leaveTangent=1.9321, tangentMode=RCTM_User, time=1, value=0}
          ]}
        }}
       CurveIndex=&RandomRangeFloat(Minimum=0, Maximum=1)
      )
      Yaw:NiagaraFloat=0
      Pitch:NiagaraFloat=&Multiply_FloatByInt(Float=60, Integer=&RandomRangeInt(Minimum=0, Maximum=5))
      Roll:NiagaraFloat=0
      Rotation Rate:NiagaraFloat=1
     )
     Vector=<0 -15 3.8564>
    )) {isSetParameters=true}
   &AddVelocity(Velocity=&RandomRangeVector(Minimum=<-1 -1 -1>, Maximum=<1 1 1>), Velocity Speed Scale=5)
   /SetVariables(Particles.StartIndex:NiagaraInt32=&RandomRangeInt(Minimum=2, Maximum=20)) {isSetParameters=true, module=SetVariables_#4}
  ]
  particleUpdateScript [
   &ParticleState()
   &SubUVAnimation(
    End Frame:NiagaraInt32=&Add_Integer(A=$Particles.StartIndex, B=&RandomRangeInt(Minimum=20, Maximum=30))
    Start Frame:NiagaraInt32=$Particles.StartIndex
   )
   &GravityForce(Gravity=&RandomRangeVector(Minimum=<0 0 -6>, Maximum=<0 0 -10>))
   &Drag(Drag=&RandomRangeFloat(Minimum=2, Maximum=3), Use Linear Drag=true, Use Rotational Drag=true)
   &SolveForcesAndVelocity(Clamp Velocity=false, Limit Acceleration=false)
   &ScaleColor(
    Scale Alpha=&FloatFromCurve(
     FloatCurve=DataInterface() {propertyValues {Curve {keys [
         {arriveTangent=0.1402, interpMode=RCIM_Cubic, leaveTangent=0.1402, tangentMode=RCTM_User, time=0, value=0}
         {arriveTangent=0.317, interpMode=RCIM_Cubic, leaveTangent=0.317, tangentMode=RCTM_User, time=0.2, value=0.4}
         <at 1 0 RCIM_Cubic>
        ]}}}
     Scale Curve=&Multiply_Float(A=&RandomRangeFloat(Minimum=0.3, Maximum=0.5), B=$System.EquipTimer)
    )
    Scale RGB=<1 1 1.1>
    ScaleA=true
    ScaleRGB=true
   )
   &DynamicMaterialParameters(
    Index 0 Param 4=1
    Module.Param0WriteEnabled=true
    Module.Param1WriteEnabled=true
    Module.Param2WriteEnabled=true
    Module.Param3WriteEnabled=true
   )
   &ScaleSpriteSize(Scale Factor=&Vector2DFromFloat(Value=&FloatFromCurve(
      FloatCurve=DataInterface() {propertyValues {Curve {keys [
          {arriveTangent=2.1241, interpMode=RCIM_Cubic, leaveTangent=2.1241, tangentMode=RCTM_User, time=0, value=0.2}
          {arriveTangent=0.8108, interpMode=RCIM_Cubic, leaveTangent=0.8108, tangentMode=RCTM_User, time=1, value=1.5}
         ]}}}
      Scale Curve=&MakeCustomFloatFromBool(Bool=$User.bPickaxeEquipped, True Float=0, False Float=1)
     )))
  ]
  rendererClasses [&NiagaraSpriteRendererProperties]
  renderers [{rendererClass=&NiagaraSpriteRendererProperties}]
 }
 Glow {
  emitterUpdateScript [
   &EmitterState(
    Inactive Response=::NewEnumerator1
    Life Cycle Mode=::NewEnumerator1
    Loop Behavior=::NewEnumerator1
    Loop Duration Mode=::NewEnumerator1
    UseLoopDelay=false
   )
   &SpawnBurst_Instantaneous(Spawn Count=11, Spawn Probability=HlslExpression() {hlslExpression=1.0}, Use Spawn Probability=false, UseLoopCountLimit=false)
  ]
  particleSpawnScript [
   &PartitionParticles(ID to Partition:NiagaraInt32=&ReturnExecIndex(), Number Of Partitions=11)
   /SetVariables(Particles.VisibilityTag=$Emitter.SpawnGroup) {isSetParameters=true}
   &InitializeParticle(
    Color=&MultiplyLinearColor(
     Linear Color=&MergeHelper_ColorSwitch(
      0=&RandomRangeLinearColor(Minimum=<rgba 0 1 0.7013 0.5>, Maximum=<rgba 0 1 0.8649 0.5>)
      1=<rgba 1 1 1 1>
      2=?
      3=?
      Index=$Particles.VisibilityTag
      Num=::NewEnumerator0
     )
     Scale Factor=&MakeVector4_FromLinearColor(LinearColor=$System.TOD)
    )
    Mesh Scale Mode=::NewEnumerator4
    Sprite Size Mode=::NewEnumerator3
    Uniform Sprite Size=&MakeCustomFloatFromBool(
     Bool=&SetBoolByIntComparison(A=$Particles.PartitionParticles.Partition, B=6, Comparison Type=::NewEnumerator4)
     True Float=&RandomRangeFloat(Minimum=12, Maximum=12)
     False Float=&RandomRangeFloat(Minimum=4, Maximum=5)
    )
    UsePositionOffset=false
    Write Material Random=true
   )
   &SocketLocation(
    Orientation Sampling=::NewEnumerator1
    SamplePosOffset=true
    Sampled Position Offset=&MakeCustomVectorFromBool(
     Bool=&SetBoolByIntComparison(A=$Particles.PartitionParticles.Partition, B=10, Comparison Type=::NewEnumerator4)
     True Vector=<13 -11.3845 1.7026>
     False Vector=&MergeHelper_VectorSwitch(
      0=<-3.0304 -16.7171 -0.0519>
      1=<-3.6471 -16.7171 5.2486>
      2=<-1.4391 -16.7171 8.5169>
      3=<3.467 -16.7171 6.2369>
      4=<5.1621 -16.7171 2.7792>
      5=<0.8391 -16.7171 -0.3513>
      6=<0 -15 4>
      7=<13 -6.2 1.7026>
      8=<-13.1265 -6.5 1.7026>
      9=<-13.1265 -11.5 1.7026>
      Index=$Particles.PartitionParticles.Partition
      Num=::NewEnumerator8
     )
    )
    Socket Reader=$System.Socket Reader
    Socket Sampling Mode=::NewEnumerator3
    VelocitySpeedLimit=false
   )
  ]
  particleUpdateScript [
   &ParticleState(Kill Particles When Lifetime Has Elapsed=false, Let Infinitely Lived Particles Die When Emitter Deactivates=true)
   &ScaleColor(
    Scale Alpha=&Multiply_Float(
     A=&MakeCustomFloatFromBool(
      Bool=&SetBoolByIntComparison(A=$Particles.PartitionParticles.Partition, B=6, Comparison Type=::NewEnumerator0)
      True Float=&FN_FrontEndFloatModifier(Gameplay=1, Front End=$System.EquipTimer)
      False Float=$System.EquipTimer
     )
     B=&MergeHelper_FloatSwitch(
      0=&Multiply_Float(
       A=0.3
       B=&Multiply_Float(
        A=1
        B=&Waveform(
         Global Input Position=&RandomRangeFloat(Minimum=0, Maximum=100)
         "Waveform [1]"=::NewEnumerator12
         "[1] Frequency"=6
         "[1] Amplitude Min/Max"=<2 2.5>
        )
       )
      )
      1=1
      2=?
      3=?
      Index=$Particles.VisibilityTag
      Num=::NewEnumerator0
     )
    )
    Scale RGB=&MakeVectorFromLinearColorRGB(Color=&Multiply_LinearColorByFloat(Linear Color=$System.TOD, Float=3))
    ScaleA=true
    ScaleRGB=true
   )
   &SocketLocation(
    Interpolate Socket Sampling=true
    Orientation Sampling=::NewEnumerator1
    Position Sampling=::NewEnumerator1
    SamplePosOffset=true
    Sampled Position Coordinate Space:ENiagaraCoordinateSpace=::World
    Socket Index=1
    Socket Reader=$System.Socket Reader
    Socket Sampling Mode=::NewEnumerator3
    VelocitySpeedLimit=false
   ) {module=SocketLocation001}
   /SetVariables(Particles.DirectionDot:NiagaraFloat=&ClampFloat(Float=&DotProduct(
      Vector A=&NormalizeVector(
       Vector To Normalize=&Subtract_Position(A=$Engine.Emitter.SimulationPosition, B=$Particles.Position)
       Safe Normalize=true
       Safe Fallback Threshold=0
       Safe Fallback Vector=<0 0 0.001>
      )
      Vector B=&ReturnCameraVector(Camera Vectors=::NewEnumerator1, Coordinate Space=::Simulation)
     ))) {isSetParameters=true}
   &ScaleSpriteSize(
    Scale Sprite Size Mode=::NewEnumerator0
    Uniform Scale Factor=&MakeCustomFloatFromBool(
     Bool=$User.bPickaxeEquipped
     True Float=&MakeCustomFloatFromBool(
      Bool=&SetBoolByIntComparison(A=$Particles.PartitionParticles.Partition, B=6, Comparison Type=::NewEnumerator0)
      True Float=1
      False Float=0
     )
     False Float=1
    )
   )
   &CameraOffset(Camera Offset Amount=&MergeHelper_FloatSwitch(
     0=&Lerp_Float(
      A=0
      B=2
      Alpha=&FloatFromCurve(FloatCurve=DataInterface() {propertyValues {Curve {keys [<at 0.6 0>,<at 0.8 1>]}}}, CurveIndex=$Particles.DirectionDot)
     )
     1=0
     2=?
     3=?
     Index=$Particles.VisibilityTag
     Num=::NewEnumerator0
    ))
  ]
  rendererClasses [&NiagaraSpriteRendererProperties]
  renderers [{rendererClass=&NiagaraSpriteRendererProperties}]
 }
}
summary {
 systemName=NS_Backpack_TicketPoke_Idle
 emitters [
  {emitterName=FrostSmokeLocal, rendererClasses [&NiagaraSpriteRendererProperties]}
  {emitterName=Glow, rendererClasses [&NiagaraSpriteRendererProperties]}
  {emitterName=FrostSmoke, rendererClasses [&NiagaraSpriteRendererProperties]}
 ]
 userVariables [
  User.bIsFrontEnd {type=NiagaraBool}
  User.bIsFrontEndPreview {type=NiagaraBool}
  User.bPickaxeEquipped {type=NiagaraBool}
  User.TimeStamp {type=NiagaraFloat, defaultValue=0, description=""}
 ]
}
