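# Animation Blueprint for the TicketPoke backpack cosmetic that adds spring-driven secondary motion to the left and right dragon jaw bones.
# The cached input pose feeds two chains of Anim Dynamics nodes, each simulating both jaws with damped angular springs, gravity and one-sided angular limits.
# Blend Poses by bool selects between the chains on bIsInFrontEnd, with the front-end chain using larger physics boxes and different joint offsets.
# The event graph is empty.
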AnimBlueprint /BRCosmetics/Accessories/FORT_Backpacks/Backpack_M_MED_TicketPoke/Meshes/M_MED_TicketPoke_Pack_AnimBP
target_skeleton=&/M_MED_TicketPoke_Pack_Skeleton
graphs [
 AnimGraph {nodes [
   n1 = Anim Dynamics - Bone: L_jaw_dragon(ComponentPose=Local To Component(LocalPose=Use cached pose 'input cached'()), Alpha=1, ExternalForce="0,0,0") {
    angularDampingOverride=1
    angularSpringConstant=5800
    bAngularSpring=true
    bOverrideAngularDamping=true
    bUseGravityOverride=true
    boundBone {boneName=L_jaw_dragon}
    componentAppliedLinearAccClamp=<800 800 800>
    componentLinearAccScale=<0.2 0.2 0.2>
    componentLinearVelScale=<0.2 0.2 0.2>
    gravityOverride {z=-980}
    physicsBodyDefinitions [{
      boundBone {boneName=L_jaw_dragon}
      boxExtents=<20 20 15>
      constraintSetup {angularLimitsMax=<0 0 0>, angularLimitsMin=<0 -9 0>, angularTarget=<1 0 0>}
      localJointOffset=<1.8 0 0.3>
     }]
   }
   n2 = Anim Dynamics - Bone: R_jaw_dragon(ComponentPose=@n1, Alpha=1, ExternalForce="0,0,0") {
    angularDampingOverride=1
    angularSpringConstant=5000
    bAngularSpring=true
    bOverrideAngularDamping=true
    bUseGravityOverride=true
    boundBone {boneName=R_jaw_dragon}
    componentAppliedLinearAccClamp=<800 800 800>
    componentLinearAccScale=<0.2 0.2 0.2>
    componentLinearVelScale=<0.2 0.2 0.2>
    gravityOverride {z=-980}
    physicsBodyDefinitions [{
      boundBone {boneName=R_jaw_dragon}
      boxExtents=<20 20 15>
      constraintSetup {angularLimitsMax=<0 9 0>, angularLimitsMin=<0 0 0>, angularTarget=<1 0 0>}
      localJointOffset=<1.5 0 0.1>
     }]
   }
   n3 = Anim Dynamics - Bone: L_jaw_dragon(ComponentPose=Local To Component(LocalPose=Use cached pose 'input cached'()), Alpha=1, ExternalForce="0,0,0") {
    angularDampingOverride=1
    angularSpringConstant=5800
    bAngularSpring=true
    bOverrideAngularDamping=true
    bUseGravityOverride=true
    boundBone {boneName=L_jaw_dragon}
    componentAppliedLinearAccClamp=<800 800 800>
    componentLinearAccScale=<0.2 0.2 0.2>
    componentLinearVelScale=<0.2 0.2 0.2>
    gravityOverride {z=-980}
    physicsBodyDefinitions [{
      boundBone {boneName=L_jaw_dragon}
      boxExtents=<14 15 8>
      constraintSetup {angularLimitsMax=<0 0 0>, angularLimitsMin=<0 -9 0>, angularTarget=<1 0 0>}
      localJointOffset=<1.4 0 0>
     }]
   }
   n4 = Anim Dynamics - Bone: R_jaw_dragon(ComponentPose=@n3, Alpha=1, ExternalForce="0,0,0") {
    angularDampingOverride=1
    angularSpringConstant=5000
    bAngularSpring=true
    bOverrideAngularDamping=true
    bUseGravityOverride=true
    boundBone {boneName=R_jaw_dragon}
    componentAppliedLinearAccClamp=<800 800 800>
    componentLinearAccScale=<0.2 0.2 0.2>
    componentLinearVelScale=<0.2 0.2 0.2>
    gravityOverride {z=-980}
    physicsBodyDefinitions [{
      boundBone {boneName=R_jaw_dragon}
      boxExtents=<15 14 9>
      constraintSetup {angularLimitsMax=<0 9 0>, angularLimitsMin=<0 0 0>, angularTarget=<1 0 0>}
      localJointOffset=<1.4 0 0>
     }]
   }
   Output Pose(Result=Blend Poses by bool(BlendPose_0=Component To Local(ComponentPose=@n2), BlendPose_1=Component To Local(ComponentPose=@n4), bActiveValue=Get bIsInFrontEnd()) {blendType=Linear}) {layerGroup=None}
   Save cached pose 'input cached'(Pose=Input Pose()) {cacheName=input cached}
  ]}
 EventGraph {links [], nodes []}
]
variables ["TicketPoke  pickaxe" {type=boolean}]
